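# Translucent unlit overlay material for the CoconutShell backpack, faking lit glass through emissive and opacity.
# A light vector from custom View directional-light reads, falling back to the sun parameter collection, drives fake specular highlights.
# Tech-art and texture masks place Fresnel rims, inner glow and polar-mapped panning dots.
# Time-of-day functions tint color and intensity, with sine-pulsed bioluminescent night emissive and front-end overrides.
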Material /BRCosmetics/Accessories/FORT_Backpacks/Backpack_M_MED_CoconutShell/Materials/M_CoconutShell_OverlayTrans
graph {
 properties {
  bCastRayTracedShadows=false
  bUseMaterialAttributes=true
  bUsedWithMeshDeformer=true
  bUsedWithMorphTargets=true
  bUsedWithSkeletalMesh=true
  blendMode=BLEND_Translucent
  pixelDepthOffsetMode=PDOM_Legacy
  shadingModel=MSM_Unlit
  twoSided=true
  usedShadingModels=Unlit
 }
 nodes [
  AtmosphericLightVector()
  n1 = VectorParameter() {group=Toon_Shading, parameterName=ScreenSpaceLightVector, defaultValue=<rgba 1 1 -1 1>}
  Composite!(LightVector=@n1.RGB) {subgraphName=ProjectScreenAlighned}
  Constant()
  Constant()
  n2 = NamedRerouteUsage()
  n3 = Multiply(A=@n2, B=VectorParameter() {group=FakeGlass, parameterName=EveningTint, defaultValue=<rgba 0.6908 0.5826 1 1>}.RGB)
  n4 = Multiply(A=@n2, B=VectorParameter() {group=FakeGlass, parameterName=NightTint, defaultValue=<rgba 0.5417 0.5626 1 1>}.RGB)
  MaterialFunctionCall(In=SetMaterialAttributes!(
    MaterialAttributes
    Emissive Color=MaterialFunctionCall(Color_Morning=@n2, Color_Day=@n2, Color_Evening=@n3, Color_Night=@n4) {materialFunction=&Color_Time_of_Day}.Result
    Opacity=NamedRerouteUsage()
   ) {attributeSetTypes [#1, #2]}) {materialFunction=&MF_TechArtOverrides_accessories}
  MaterialFunctionCall() {materialFunction=/Game/Global/Materials/MaterialGym/Materials/SoftGlowOutlines/MF_SoftGlowOutlines}
  n5 = MaterialFunctionCall() {materialFunction=&MF_CharacterParent_2023_PreFX}
  n6 = GetMaterialAttributes(None=@n5.Result) {attributeGetTypes [#3, #4, #1]}
  n7 = VectorParameter() {group=FakeGlass, parameterName=BaseColorTint, channelNames {a=BaseColorIntensity, b="", g="", r=""}, defaultValue=<rgba 1 1 1 1>}
  n8 = Multiply(A=NamedRerouteDeclaration(None=@n6.Base Color) {name=BaseColorIN, nodeColor=<rgba 0 0.0625 1 1>}, B=Multiply(A=@n7, B=@n7))
  n9 = LinearInterpolate(
   A=Multiply(A=ScalarParameter() {defaultValue=0.8, parameterName=BaseColorLow}, B=@n8) {constB=1.2}
   B=NamedRerouteDeclaration(None=@n8) {name=BaseColorTinted, nodeColor=<rgba 0.125 0 1 1>}
   Alpha=NamedRerouteUsage()
  )
  Multiply(A=@n9)
  n10 = NamedRerouteUsage()
  n11 = VectorParameter() {parameterName=OffsetPivot, defaultValue=<rgba 0 -10.9944 0 1>}
  n12 = Divide(
   A=ComponentMask(None=Subtract(A=@n10, B=@n11.RGB)) {g=true, r=true}
   B=ScalarParameter() {defaultValue=50, group=FakeGlass, parameterName=TilllingUnits}
  )
  n13 = MaterialFunctionCall!(UV Coord=Add(A=@n12) {constB=0.5}, LenghtScale, RadialScale) {materialFunction=/Game/Global/MaterialFunctions/TechArt/MF_PolarCoordinates}
  n14 = ComponentMask(None=@n13.Result) {r=true}
  n15 = NamedRerouteDeclaration(None=Saturate(None=SmoothStep(Value=@n14) {constMax=0.15})) {name=RadialMask, nodeColor=<rgba 1 0 0.8203 1>}
  Multiply(A=NamedRerouteUsage(), B=@n15)
  n16 = CameraVectorWS()
  n17 = ComponentMask(None=Custom() {code="return View.DirectionalLightColor;", description=DirectionalLightColor, outputType=CMOT_Float4, inputs []}) {b=true, g=true, r=true}
  n18 = Custom() {code="return View.DirectionalLightDirection;", description=DirectionalLightDirection, inputs []}
  n19 = CollectionParameter() {
   collection=&FortniteMaterialParameters
   parameterName=SunlightDirectionalVector
  }
  n20 = ComponentMask(None=@n19) {b=true, g=true, r=true}
  n21 = If(A=DotProduct(A=@n18, B=@n18), "A > B"=@n18, "A == B"=@n20, "A < B"=@n20) {desc=check if DirectionalLight exists}
  n22 = ComponentMask(None=NamedRerouteDeclaration(None=If(A=@n17, "A > B"=@n21, "A == B"=@n20, "A < B"=@n20) {desc=check if DirectionalLight color is Black}) {name=LightVec, nodeColor=<rgba 0 1 0.9297 1>}) {b=true, g=true, r=true}
  n23 = MaterialFunctionCall(Float3=@n22) {materialFunction=&BreakOutFloat3Components}
  n24 = MaterialFunctionCall(X=@n23.R, Y=@n23.R, Z=Max(A=@n23.B) {constB=0.2}) {materialFunction=&MakeFloat3}
  n25 = MaterialFunctionCall(Float3=@n22) {materialFunction=&BreakOutFloat3Components}
  n26 = MaterialFunctionCall(X=@n25.R, Y=@n25.R, Z=Max(A=@n25.B) {constB=0.9}) {materialFunction=&MakeFloat3}
  n27 = ScalarParameter() {group=FakeGlass, parameterName=IsFrontEnd}
  n28 = ComponentMask(None=LinearInterpolate(
    A=Normalize(VectorInput=@n24.Result)
    B=Normalize(VectorInput=@n26.Result)
    Alpha=NamedRerouteDeclaration(None=@n27) {name=FrontEnd, nodeColor=<rgba 0.8047 1 0 1>}
   )) {b=true, g=true, r=true}
  n29 = DotProduct(A=Normalize(VectorInput=Add(A=@n16, B=@n28)), B=Normalize(VectorInput=PixelNormalWS())) {desc=Blinn}
  n30 = DotProduct(A=ReflectionVectorWS(), B=Normalize(VectorInput=@n28)) {desc=Phong}
  n31 = ScalarParameter() {defaultValue=1, group=FakeGlass, parameterName=BlinnOrPhong}
  n32 = Power(
   Base=Max(A=LinearInterpolate(A=@n29, B=@n30, Alpha=@n31)) {constB=0}
   Exp=ScalarParameter() {defaultValue=20, group=FakeGlass, parameterName=SpecPow}
  )
  Multiply(A=@n32, B=NamedRerouteUsage())
  n33 = VectorParameter() {parameterName=PreviousFramePos, defaultValue=<rgba 0 0 10 1>}
  Multiply(A=PreviousFrameSwitch(Current Frame=Constant3Vector(), Previous Frame=ComponentMask(None=@n33.RGB) {b=true, g=true, r=true})) {constB=100}
  NamedRerouteDeclaration(None=@n6.Normal) {name=Normal_IN, nodeColor=<rgba 0 1 0.8594 1>}
  NamedRerouteDeclaration(None=Sign(None=@n12)) {name=CenterDebug, nodeColor=<rgba 0 0.625 1 1>}
  n34 = OneMinus(None=Fresnel() {baseReflectFraction=0, exponent=1})
  n35 = ScalarParameter() {defaultValue=1.5, group=SoftGlow, parameterName=FresnelOutline_Power}
  NamedRerouteDeclaration(None=Saturate(None=Power(Base=@n34, Exp=@n35) {constExponent=3})) {name=SimpleFresnel, nodeColor=<rgba 0 0.4141 1 1>}
  MaterialExpressionReroute_12 = Reroute(None=@n1.RGB)
  n36 = Normalize(VectorInput=@MaterialExpressionReroute_12.LightVector)
  n37 = ComponentMask(None=@n36) {r=true}
  n38 = Constant3Vector() {constant=<rgba 0 1 0 0>}
  n39 = Normalize(VectorInput=WorldPosition() {worldPositionShaderOffset=WPT_CameraRelative}.XYZ)
  n40 = Normalize(VectorInput=CrossProduct(A=Transform(None=@n38) {transformSourceType=TRANSFORMSOURCE_View}, B=@n39))
  n41 = ComponentMask(None=@n36) {g=true}
  n42 = CrossProduct(A=@n39, B=@n40)
  n43 = Add(A=Add(A=Multiply(A=@n37, B=@n40), B=Multiply(A=@n41, B=@n42)), B=Multiply(A=ComponentMask(None=@n36) {b=true}, B=@n39))
  MaterialExpressionReroute_14 = Reroute(None=@n43)
  n44 = ComponentMask(None=Custom() {code="return View.DirectionalLightColor;", description=DirectionalLightColor, outputType=CMOT_Float4, inputs []}) {b=true, g=true, r=true}
  n45 = Custom() {code="return View.DirectionalLightDirection;", description=DirectionalLightDirection, inputs []}
  n46 = CollectionParameter() {
   collection=&FortniteMaterialParameters
   parameterName=SunlightDirectionalVector
  }
  n47 = ComponentMask(None=@n46) {b=true, g=true, r=true}
  n48 = If(A=DotProduct(A=@n45, B=@n45), "A > B"=@n45, "A == B"=@n47, "A < B"=@n47) {desc=check if DirectionalLight exists}
  n49 = StaticSwitchParameter(
   True=@MaterialExpressionReroute_14.Result
   False=MaterialFunctionCall!(In=If(A=@n44, "A > B"=@n48, "A == B"=@n47, "A < B"=@n47) {desc=check if DirectionalLight color is Black}) {materialFunction=/Game/Packages/Fortress_SharedMaterials/Cinematics/Anime/MF_CINE_AnimeLightDirOverride}.Result
  ) {
   desc="Use a Screen / View alight light vector. \r\nOverrides the UseDirectionalLightFunction!"
   group=Toon_Shading
   parameterName=UseScreenSpaceLightDirection
   sortPriority=0
  }
  n50 = MaterialFunctionCall(Float3=@n49) {materialFunction=&BreakOutFloat3Components}
  n51 = ScalarParameter() {
   desc="0.0 = the light vector is locked on the XY plane.\r\n1.0 = the actual light vector is used  \r\n(for Faces use a blend value of 0,0)"
   group=Toon_Shading
   parameterName=LockLightZ_Blend
   sliderMax=1
  }
  n52 = MaterialFunctionCall(X=@n50.R, Y=@n50.R, Z=Multiply(A=@n50.B, B=@n51)) {materialFunction=&MakeFloat3}
  n53 = StaticSwitchParameter(True=LinearInterpolate(A=@n49, B=Normalize(VectorInput=@n52.Result), Alpha=VertexColor().A), False=@n49) {
   desc="Lock the Z axis of the Light Vector\r\nis driven by the VertexColor A channel (0 = Regular LightVector, 1= Locked LIghtVector)"
   group=Toon_Shading
   parameterName=UseLockLightZVector
   sortPriority=0
  }
  NamedRerouteDeclaration(None=ComponentMask(None=@n53) {b=true, g=true, r=true}) {name=LightVector, nodeColor=<rgba 0 1 0.8634 1>}
  NamedRerouteDeclaration(None=@n5.Pre skinned Local Position) {name=PreskinnedPOS, nodeColor=<rgba 0.5938 0 1 1>}
  n54 = MaterialFunctionCall(Float2=@n13.Result) {materialFunction=&BreakOutFloat2Components}
  n55 = MaterialFunctionCall(X=@n54.R, Y=Multiply(A=@n54.G, B=ScalarParameter() {defaultValue=0.4, parameterName=V_Tiling})) {materialFunction=&MakeFloat2}
  n56 = ScalarParameter() {defaultValue=0.9, parameterName=DotsTilling_Pink}
  n57 = Multiply(A=Time(), B=ScalarParameter() {defaultValue=0.07, group=FakeGlass, parameterName=FlowSpeed02}) {constB=0.05}
  n58 = Constant2Vector() {r=-1}
  n59 = TextureSampleParameter2D(UVs=Add(A=Multiply(A=@n55.Result, B=@n56), B=Multiply(A=@n57, B=@n58))) {
   group=FakeGlass
   parameterName=GlobalDots
   samplerSource=SSM_Wrap_WorldGroupSettings
   samplerType=SAMPLERTYPE_Masks
   texture=/BRCosmetics/Accessories/FORT_Backpacks/Backpack_M_MED_CoconutShell/Textures/T_GlobalDots
  }
  NamedRerouteDeclaration(None=@n59.R) {name=Spot, nodeColor=<rgba 1 0 0.0234 1>}
  n60 = ScalarParameter() {defaultValue=0.5, parameterName=SineSpeed}
  n61 = Time()
  n62 = ConstantBiasScale(None=Sine(None=Multiply(A=@n60, B=@n61)))
  NamedRerouteDeclaration(None=@n62) {name=TimeSine, nodeColor=<rgba 0 0.3437 1 1>}
  NamedRerouteDeclaration(None=@n6.Emissive Color) {name=EM_IN, nodeColor=<rgba 0 1 0.1797 1>}
  n63 = Constant()
  n64 = Constant() {r=0.5}
  n65 = Constant() {r=1}
  n66 = ScalarParameter() {group=FrontEnd, parameterName=FrontEndAlpha}
  NamedRerouteDeclaration(None=LinearInterpolate(
    A=MaterialFunctionCall(Value_Morning=@n63, Value_Day=@n63, Value_Evening=@n64, Value_Night=@n65) {materialFunction=&Scalar_Time_of_Day}.Result
    B=Multiply(A=@n66)
    Alpha=NamedRerouteUsage()
   )) {name=Intensity_TOD, nodeColor=<rgba 0 0.6953 1 1>}
  NamedRerouteDeclaration(None=Constant()) {name=OpacityADD_Night, nodeColor=<rgba 1 0.2578 0 1>}
  n67 = TextureSampleParameter2D() {
   group=FakeGlass_Night
   parameterName=NightEM
   texture=/BRCosmetics/Accessories/FORT_Backpacks/Backpack_M_MED_CoconutShell/Textures/T_CoconutShell_Backpack_Night_E
  }
  n68 = Multiply(A=ScalarParameter() {defaultValue=0.2, parameterName=GeneralEM_Night}, B=@n67.RGB)
  n69 = TextureSampleParameter2D() {
   group=FakeGlass_Night
   parameterName=Bio_EM
   texture=/BRCosmetics/Accessories/FORT_Backpacks/Backpack_M_MED_CoconutShell/Textures/T_CoconutNight_Backpack_FX
  }
  n70 = Time()
  n71 = ScalarParameter() {defaultValue=0.4, parameterName=BioSineSpeed}
  n72 = ConstantBiasScale(None=Sine(None=Multiply(A=@n70, B=@n71)))
  n73 = LinearInterpolate(A=@n69.R, B=@n69.R, Alpha=@n72)
  n74 = VectorParameter() {group=FakeGlass_Night, parameterName=NightEM01, channelNames {a=Intensity, b="", g="", r=""}, defaultValue=<rgba 0.2927 0.8723 1 2>}
  n75 = Multiply(A=@n74, B=@n74)
  n76 = LinearInterpolate(Alpha=OneMinus(None=@n72)) {constA=1, constB=1.6}
  n77 = VectorParameter() {group=FakeGlass_Night, parameterName=NightEM02, channelNames {a=Intensity, b="", g="", r=""}, defaultValue=<rgba 1 0.6062 0.7855 1>}
  n78 = Multiply(A=@n69.G, B=Multiply(A=@n77, B=@n77))
  n79 = Add(A=@n68, B=Add(A=Multiply(A=@n73, B=@n75), B=Multiply(A=@n76, B=@n78)))
  n80 = Multiply(A=Time(), B=ScalarParameter() {defaultValue=0.5, parameterName=BioSineSpeed01})
  n81 = LinearInterpolate(Alpha=ConstantBiasScale(None=Sine(None=@n80))) {constA=0.5, constB=2}
  NamedRerouteDeclaration(None=Multiply(A=Multiply(A=@n79, B=@n81), B=NamedRerouteUsage())) {name=NightEM, nodeColor=<rgba 0 0.0625 1 1>}
  n82 = TextureSampleParameter2D() {
   group=NewFakeTransOverlay
   parameterName=AreaMask
   samplerType=SAMPLERTYPE_Masks
   texture=/BRCosmetics/Accessories/FORT_Backpacks/Backpack_M_MED_CoconutShell/Textures/T_CoconutNight_New_Backpack_AreaMask
  }
  NamedRerouteDeclaration(None=@n82.R) {name=LeafArea, nodeColor=<rgba 0.4766 0 1 1>}
  NamedRerouteDeclaration(None=@n15) {name=FresnelNoise, nodeColor=<rgba 0 1 0.9531 1>}
  NamedRerouteDeclaration(None=@n82.B) {name=CenterArea, nodeColor=<rgba 1 0.6328 0 1>}
  n83 = VectorParameter() {group=FakeGlass, parameterName=FresnelColor01, defaultValue=<rgba 0.125 0.8 1 1>}
  NamedRerouteDeclaration(None=@n83.RGB) {name=Fresnel01Color, nodeColor=<rgba 0 0.2969 1 1>}
  NamedRerouteDeclaration(None=@n68) {name=DefaultAdd, nodeColor=<rgba 1 0 0.3281 1>}
  NamedRerouteDeclaration(None=@n82.G) {name=LightingArea, nodeColor=<rgba 0 1 0.9063 1>}
  NamedRerouteDeclaration(None=ChannelMaskParameter(None=@n5.Tech Art Mask) {group=FakeGlass, parameterName=UseFakeGlass}.R) {name=TechArtMask, nodeColor=<rgba 1 0.9375 0 1>}
  n84 = ChannelMaskParameter(None=@n5.Tech Art Mask) {group=FakeGlass, maskChannel=Blue, parameterName=InnerEM, defaultValue=<rgba 0 0 1 0>}
  NamedRerouteDeclaration(None=Multiply(A=NamedRerouteUsage(), B=OneMinus(None=@n84.R))) {name=LightArea, nodeColor=<rgba 1 0.5859 0 1>}
  n85 = TextureSampleParameter2D() {
   group=FakeGlass
   parameterName=DynamicMask
   texture=/BRCosmetics/Accessories/FORT_Backpacks/Backpack_M_MED_CoconutShell/Textures/T_CoconutNight_New_Backpack_Gradient
  }
  n86 = OneMinus(None=@n85.R)
  NamedRerouteDeclaration(None=@n86) {name=DynamicMask, nodeColor=<rgba 0.875 0 1 1>}
  NamedRerouteDeclaration(None=@n85.G) {name=AreaMask, nodeColor=<rgba 1 0 0.7969 1>}
  n87 = VectorParameter() {group=FakeGlass, parameterName=FresnelColorInner, defaultValue=<rgba 0.0521 0.667 1 1>}
  NamedRerouteDeclaration(None=@n87.RGB) {name=InnerColor, nodeColor=<rgba 0 1 0.8828 1>}
  NamedRerouteDeclaration(None=ScalarParameter() {group=FrontEnd, parameterName=IsInFrontEnd}) {name=FrontEndOverride, nodeColor=<rgba 1 0.3047 0 1>}
  n88 = ComponentMask(None=NamedRerouteUsage()) {b=true, g=true, r=true}
  n89 = PixelNormalWS()
  NamedRerouteDeclaration(None=Saturate(None=DotProduct(A=@n88, B=@n89))) {name=FakeDiffuse, nodeColor=<rgba 1 0 0.4687 1>}
  NamedRerouteDeclaration(None=@n6.MaterialAttributes) {name=MA_IN, nodeColor=<rgba 0.5703 0 1 1>}
  NamedRerouteDeclaration(None=ChannelMaskParameter(None=@n5.Tech Art Mask) {group=FakeGlass, maskChannel=Green, parameterName=MaskNoise, defaultValue=<rgba 0 1 0 0>}.R) {name=OneFaceMask, nodeColor=<rgba 0.4766 1 0 1>}
  n90 = ScalarParameter() {defaultValue=0.5, group=FakeGlass, parameterName=Fresnel_Intensity03}
  n91 = Multiply(
   A=VectorParameter() {group=FakeGlass, parameterName=FresnelColor03, defaultValue=<rgba 1 1 1 1>}.RGB
   B=SmoothStep(Value=Fresnel() {baseReflectFraction=0, exponent=1}) {constMin=0.6}
  )
  n92 = VectorParameter() {group=FakeGlass, parameterName=BlueTint, defaultValue=<rgba 1 1 1 1>}
  n93 = LinearInterpolate(
   A=@n92.RGB
   B=Multiply(A=@n92.RGB, B=ScalarParameter() {defaultValue=4, parameterName=BlueTint_Strength}) {constB=2}
   Alpha=NamedRerouteUsage()
  )
  n94 = Multiply(
   A=Fresnel(ExponentIn=ScalarParameter() {defaultValue=1.5, parameterName=OutLineStrength_POW}) {baseReflectFraction=0, exponent=1}
   B=NamedRerouteUsage()
  )
  n95 = ScalarParameter() {defaultValue=2, group=FakeGlass, parameterName=Fresnel_Strength_Intensity}
  n96 = LinearInterpolate(A=@n95, B=Multiply(A=@n95) {constB=5}, Alpha=NamedRerouteUsage())
  n97 = VectorParameter() {group=FakeGlass, parameterName=RimColor01, defaultValue=<rgba 0 0.2085 1 1>}
  n98 = LinearInterpolate(Alpha=NamedRerouteUsage()) {constA=1, constB=0.2}
  n99 = VectorParameter() {group=FakeGlass, parameterName=HighLight, channelNames {a=HighLightIntensity, b="", g="", r=""}, defaultValue=<rgba 1 1 1 1>}
  n100 = NamedRerouteDeclaration(None=Multiply(A=@n99, B=@n99)) {name=HighLight, nodeColor=<rgba 1 0 0.5859 1>}
  n101 = LinearInterpolate(A=@n100, B=Multiply(A=@n100) {constB=0.7}, Alpha=NamedRerouteUsage())
  n102 = TextureSampleParameter2D() {parameterName=Flat_Normal, samplerType=SAMPLERTYPE_Normal, texture=&FlatNormalBlue}
  n103 = NamedRerouteUsage()
  n104 = Multiply(
   A=MaterialFunctionCall(Vector 3=Subtract(A=@n102.RGB, B=@n103)) {materialFunction=/Engine/Functions/Engine_MaterialFunctions02/Utility/VectorLength}.V3 Length
   B=ScalarParameter() {defaultValue=0.3, group=FakeGlass, parameterName=NormalEdgingMask, sliderMax=1}
  )
  n105 = Add(A=NamedRerouteUsage(), B=LinearInterpolate(A=@n9, B=Multiply(A=@n98, B=@n101), Alpha=Add(A=@n104, B=@n32)))
  n106 = NamedRerouteUsage()
  n107 = ScalarParameter() {defaultValue=1, group=FakeGlass, parameterName=Fresnel_Intensity01}
  n108 = Multiply(A=@n107) {constB=1.2}
  n109 = Multiply(
   A=MaterialFunctionCall(Value_Morning=@n107, Value_Day=@n107, Value_Evening=@n108, Value_Night=@n108) {materialFunction=&Scalar_Time_of_Day}.Result
   B=Multiply(A=@n83.RGB, B=Fresnel() {baseReflectFraction=0, exponent=1})
  )
  n110 = Add(
   A=Multiply(A=NamedRerouteDeclaration(None=@n94) {name=CenterOpacity, nodeColor=<rgba 0.0312 1 0 1>}, B=Multiply(A=@n87.RGB, B=@n96) {constB=20})
   B=Add(A=LinearInterpolate(A=@n97.RGB, B=@n105, Alpha=@n106), B=Multiply(A=@n109, B=@n15))
  )
  n111 = ScalarParameter() {defaultValue=5, group=FakeGlass, parameterName=Fresnel_Intensity02}
  n112 = Multiply(A=@n111) {constB=3}
  n113 = NamedRerouteUsage()
  n114 = Fresnel() {baseReflectFraction=0, exponent=1}
  n115 = Multiply(
   A=Multiply(
    A=VectorParameter() {group=FakeGlass, parameterName=FresnelColor02, defaultValue=<rgba 0.378 0.8014 1 1>}.RGB
    B=LinearInterpolate(A=@n111, B=@n112, Alpha=@n113)
   )
   B=Power(Base=OneMinus(None=@n114)) {constExponent=3}
  )
  n116 = LinearInterpolate(A=@n59.B, B=@n59.B, Alpha=@n62)
  n117 = Multiply(A=Multiply(A=Multiply(A=@n115, B=@n116), B=@n86), B=NamedRerouteUsage())
  n118 = NamedRerouteUsage()
  n119 = Multiply(
   A=NamedRerouteUsage()
   B=Multiply(A=Power(Base=@n118) {constExponent=3}, B=ScalarParameter() {defaultValue=5, parameterName=SpotIntensity}) {constB=20}
  )
  n120 = NamedRerouteUsage()
  n121 = LinearInterpolate(A=Add(A=Add(A=@n110, B=@n117), B=Multiply(A=@n119, B=@n120)), B=NamedRerouteUsage(), Alpha=NamedRerouteUsage())
  n122 = Multiply(
   A=Add(A=Multiply(A=@n90, B=@n91), B=Multiply(A=@n93, B=@n121))
   B=ScalarParameter() {defaultValue=0.8, group=FakeGlass_Override, parameterName=OveralEM_Control}
  )
  n123 = NamedRerouteUsage()
  n124 = NamedRerouteUsage()
  NamedRerouteDeclaration(None=Multiply(A=Add(A=@n122, B=@n123), B=LinearInterpolate(Alpha=@n124) {constA=0.8})) {name=EmissiveOUT, nodeColor=<rgba 1 0.2344 0 1>}
  n125 = Multiply(
   A=LinearInterpolate(
    A=ScalarParameter() {defaultValue=0.2, group=FakeGlass, parameterName=OpacityMin}
    B=ScalarParameter() {defaultValue=1, group=FakeGlass, parameterName=OpacityMax}
    Alpha=@n32
   )
   B=NamedRerouteUsage()
  )
  n126 = LinearInterpolate(
   B=LinearInterpolate(A=Saturate(None=@n125), B=ScalarParameter() {defaultValue=0.2, parameterName=LightOpacity}, Alpha=NamedRerouteUsage()) {constB=0.5}
   Alpha=NamedRerouteUsage()
  ) {constB=0}
  n127 = NamedRerouteUsage()
  NamedRerouteDeclaration(None=Saturate(None=Add(A=@n126, B=@n127))) {name=Opacity_OUT, nodeColor=<rgba 0.6875 0 1 1>}
  NamedRerouteUsage()
  Normalize(VectorInput=@n103)
  PinBase!(Result=@n43) {reroutePins [Result {expression=&self:MaterialExpressionReroute_14}]}
  PinBase!() {reroutePins [LightVector {expression=&self:MaterialExpressionReroute_12}]}
  ScalarParameter() {parameterName=BaseColorIntensity}
  SkyAtmosphereLightDirection()
 ]
}
