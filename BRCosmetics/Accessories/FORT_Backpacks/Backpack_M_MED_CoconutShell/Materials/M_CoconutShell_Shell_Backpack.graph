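# Unlit, masked, two-sided character material for the CoconutShell backpack's shell, faking tinted translucent glass.
# Base color, emissive and a Fresnel gradient mix with dot colour, tint and intensity masks between the CharacterParent PreFX and PostFX functions.
# TechArt masks and TwoSidedSign shape opacity, while ShadowReplace keeps shadows solid.
# Time-of-day drives a sine-pulsed bioluminescent night glow; some polar-flow and Fresnel branches are unconnected.
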Material /BRCosmetics/Accessories/FORT_Backpacks/Backpack_M_MED_CoconutShell/Materials/M_CoconutShell_Shell_Backpack
graph {
 properties {
  bUseMaterialAttributes=true
  bUsedWithMeshDeformer=true
  bUsedWithMorphTargets=true
  bUsedWithSkeletalMesh=true
  blendMode=BLEND_Masked
  pixelDepthOffsetMode=PDOM_Legacy
  shadingModel=MSM_Unlit
  twoSided=true
  usedShadingModels=Unlit
 }
 nodes [
  n1 = NamedRerouteUsage()
  n2 = VectorParameter() {parameterName=OffsetPivot, defaultValue=<rgba 0 -10.9944 0 1>}
  n3 = Divide(
   A=ComponentMask(None=Subtract(A=@n1, B=@n2.RGB)) {g=true, r=true}
   B=ScalarParameter() {defaultValue=30, group=FakeGlass, parameterName=TilllingUnits}
  )
  n4 = MaterialFunctionCall!(UV Coord=Add(A=@n3) {constB=0.5}, LenghtScale, RadialScale) {materialFunction=/Game/Global/MaterialFunctions/TechArt/MF_PolarCoordinates}
  n5 = Time()
  n6 = ScalarParameter() {defaultValue=0.02, group=FakeGlass, parameterName=FlowSpeed01}
  Add(A=@n4.Result, B=Multiply(A=Multiply(A=@n5, B=@n6) {constB=0.05}, B=Constant2Vector() {r=-1}))
  GetMaterialAttributes(None=MaterialFunctionCall() {materialFunction=/Game/Global/Materials/MaterialGym/Materials/SoftGlowOutlines/MF_SoftGlowOutlines}.Result) {attributeGetTypes [#1]}
  n7 = MaterialFunctionCall() {materialFunction=&MF_CharacterParent_2023_PreFX}
  n8 = GetMaterialAttributes(None=@n7.Result) {attributeGetTypes [#2, #3]}
  n9 = NamedRerouteUsage()
  n10 = ShadowReplace(Default=@n9, Shadow=Constant() {r=1})
  n11 = VectorParameter() {group=FakeGlass, parameterName=BlueTint, defaultValue=<rgba 0.6979 0.8967 1 1>}
  n12 = Multiply(A=@n11.RGB, B=ScalarParameter() {defaultValue=2, parameterName=BlueTint_Strength}) {constB=2}
  n13 = NamedRerouteUsage()
  n14 = ScalarParameter() {defaultValue=1, group=NewFakeTransOverlay, parameterName=BaseColorMulti}
  n15 = Fresnel() {baseReflectFraction=0, exponent=1}
  n16 = AppendVector(A=Multiply(A=Power(Base=@n15) {constExponent=0.2}, B=ScalarParameter() {defaultValue=2, parameterName=FresnelTilling}), B=Constant() {r=0.5})
  n17 = Add(
   A=Add(A=Multiply(A=@n8.Base Color, B=@n14), B=@n8.Emissive Color)
   B=Multiply(
    A=ScalarParameter() {defaultValue=0.2, parameterName=GradientIntensity}
    B=TextureSampleParameter2D(UVs=@n16) {parameterName=GradientTint, texture=/BRCosmetics/Accessories/FORT_Backpacks/Backpack_M_MED_CoconutShell/Textures/T_GradientSprite}.RGB
   )
  )
  n18 = VectorParameter() {parameterName=DotsColor, defaultValue=<rgba 0 0.0979 0.0922 1>}
  n19 = Multiply(A=NamedRerouteUsage()) {constB=0.3}
  n20 = ScalarParameter() {defaultValue=0.6, group=FakeGlass, parameterName=Overall_Intensity}
  n21 = Multiply(A=@n20) {constB=2}
  n22 = NamedRerouteUsage()
  n23 = Multiply(
   A=Multiply(A=LinearInterpolate(A=@n11.RGB, B=@n12, Alpha=@n13), B=LinearInterpolate(A=@n17, B=@n18.RGB, Alpha=@n19))
   B=LinearInterpolate(
    A=LinearInterpolate(A=@n20, B=@n21, Alpha=@n22)
    B=ScalarParameter() {defaultValue=0.4, group=FakeGlass, parameterName=Overall_Center}
    Alpha=NamedRerouteUsage()
   )
  )
  n24 = NamedRerouteUsage()
  MaterialFunctionCall(
   Material=SetMaterialAttributes!(MaterialAttributes=@n8.MaterialAttributes, Opacity Mask=@n10, Emissive Color=Add(A=@n23, B=@n24), Opacity=@n10) {attributeSetTypes [#4, #3, #1]}
   M=@n7.M
   Pre Skinned Local Position=@n7.M
   Distance Blend=@n7.M
   Fresnel=@n7.M
  ) {materialFunction=&MF_CharacterParent_2023_PostFX}
  n25 = Fresnel() {baseReflectFraction=0, exponent=1}
  Multiply(A=Power(Base=OneMinus(None=@n25), Exp=ScalarParameter() {defaultValue=5, parameterName=DitherFresnel}) {constExponent=5}, B=@n9)
  n26 = SmoothStep(Value=ComponentMask(None=@n4.Result) {r=true}) {constMax=0.15}
  n27 = TextureSampleParameter2D(UVs=Multiply(A=TextureCoordinate(), B=ScalarParameter() {defaultValue=0.3, parameterName=DotsTilling})) {
   group=FakeGlass
   parameterName=GlobalDots
   samplerSource=SSM_Wrap_WorldGroupSettings
   samplerType=SAMPLERTYPE_Masks
   texture=/BRCosmetics/Accessories/FORT_Backpacks/Backpack_M_MED_CoconutShell/Textures/T_GlobalDots
  }
  n28 = OneMinus(None=@n27.R)
  Multiply(A=NamedRerouteDeclaration(None=Saturate(None=@n26)) {name=RadialMask, nodeColor=<rgba 1 0 0.8203 1>}, B=@n28)
  Multiply(
   A=ScalarParameter() {defaultValue=1, group=FakeGlass, parameterName=Fresnel_Intensity01}
   B=Multiply(
    A=VectorParameter() {group=FakeGlass, parameterName=FresnelColor01, defaultValue=<rgba 0.7674 0.8948 1 1>}.RGB
    B=Fresnel() {baseReflectFraction=0, exponent=1}
   )
  )
  n29 = TextureSampleParameter2D() {
   group=NewFakeTransOverlay
   parameterName=TechArt_RE
   samplerType=SAMPLERTYPE_Masks
   texture=/BRCosmetics/Accessories/FORT_Backpacks/Backpack_M_MED_CoconutShell/Textures/T_CoconutShell_Backpack_Backpack_TechArt
  }
  n30 = NamedRerouteDeclaration(None=@n29.RGB) {name=TechArtMask, nodeColor=<rgba 0 1 0.1562 1>}
  n31 = ChannelMaskParameter(None=@n30) {group=NewFakeTransOverlay, parameterName=OpacityMask}
  NamedRerouteDeclaration(None=StaticSwitchParameter(True=OneMinus(None=@n31.R), False=@n31.R) {parameterName=InvertOpacityInput}) {name=OpacityMask, nodeColor=<rgba 1 0 0.0234 1>}
  NamedRerouteDeclaration(None=Sign(None=@n3)) {name=CenterDebug, nodeColor=<rgba 0 0.625 1 1>}
  n32 = TextureSampleParameter2D() {
   group=NewFakeTransOverlay
   parameterName=AreaMask
   samplerType=SAMPLERTYPE_Masks
   texture=/BRCosmetics/Accessories/FORT_Backpacks/Backpack_M_MED_CoconutShell/Textures/T_CoconutNight_New_Backpack_AreaMask
  }
  NamedRerouteDeclaration(None=@n32.G) {name=GreenMask, nodeColor=<rgba 0 0.8828 1 1>}
  n33 = ChannelMaskParameter(None=@n30) {group=NewFakeTransOverlay, maskChannel=Green, parameterName=OneFaceMask, defaultValue=<rgba 0 1 0 0>}
  NamedRerouteDeclaration(None=@n33.R) {name=OneFaceMask, nodeColor=<rgba 1 0 0.9141 1>}
  n34 = Constant()
  NamedRerouteDeclaration(None=MaterialFunctionCall(Value_Morning=@n34, Value_Day=@n34, Value_Evening=Constant() {r=0.5}, Value_Night=Constant() {r=1}) {materialFunction=&Scalar_Time_of_Day}.Result) {name=Intensity_TOD, nodeColor=<rgba 0 0.6953 1 1>}
  n35 = LinearInterpolate(Alpha=Step(X=TwoSidedSign())) {constA=1, constB=0}
  n36 = LinearInterpolate(A=@n35, Alpha=NamedRerouteUsage())
  n37 = NamedRerouteUsage()
  NamedRerouteDeclaration(None=Multiply(A=NamedRerouteUsage(), B=LinearInterpolate(A=@n36, B=@n35, Alpha=@n37))) {name=OpacityMask_OUT, nodeColor=<rgba 0 0.8594 1 1>}
  NamedRerouteDeclaration(None=@n7.Pre skinned Local Position) {name=PreskinnedPOS, nodeColor=<rgba 0 0.5078 1 1>}
  NamedRerouteDeclaration(None=Power(Base=@n28, Exp=ScalarParameter() {defaultValue=4, parameterName=Power_Dots}) {constExponent=3}) {name=EmissiveNoise, nodeColor=<rgba 0.8281 1 0 1>}
  n38 = Time()
  n39 = ScalarParameter() {defaultValue=0.4, parameterName=BioSineSpeed}
  n40 = ConstantBiasScale(None=Sine(None=Multiply(A=@n38, B=@n39)))
  n41 = TextureSampleParameter2D() {
   group=FakeGlass_Night
   parameterName=NightEM
   texture=/BRCosmetics/Accessories/FORT_Backpacks/Backpack_M_MED_CoconutShell/Textures/T_CoconutShell_Backpack_Night_E
  }
  n42 = Multiply(A=Multiply(A=ScalarParameter() {defaultValue=0.5, parameterName=GeneralEM_Night}, B=LinearInterpolate(Alpha=@n40) {constA=0.4}), B=@n41.RGB)
  n43 = TextureSampleParameter2D() {
   group=FakeGlass_Night
   parameterName=Bio_EM
   texture=/BRCosmetics/Accessories/FORT_Backpacks/Backpack_M_MED_CoconutShell/Textures/T_CoconutNight_Backpack_FX
  }
  n44 = VectorParameter() {group=FakeGlass_Night, parameterName=NightEM01, channelNames {a=Intensity, b="", g="", r=""}, defaultValue=<rgba 0.2927 0.8723 1 0.5>}
  n45 = VectorParameter() {group=FakeGlass_Night, parameterName=NightEM02, channelNames {a=Intensity, b="", g="", r=""}, defaultValue=<rgba 1 0.4146 0.6811 0.5>}
  n46 = Add(
   A=Multiply(A=LinearInterpolate(A=@n43.R, B=@n43.R, Alpha=@n40), B=Multiply(A=@n44, B=@n44))
   B=Multiply(A=@n43.G, B=Multiply(A=@n45, B=@n45))
  )
  NamedRerouteDeclaration(None=Multiply(A=Add(A=@n42, B=@n46), B=NamedRerouteUsage())) {name=NightEM, nodeColor=<rgba 0 0.0625 1 1>}
  NamedRerouteDeclaration(None=@n32.R) {name=LeafArea, nodeColor=<rgba 0.4766 0 1 1>}
  NamedRerouteDeclaration(None=@n32.B) {name=CenterArea, nodeColor=<rgba 1 0.6328 0 1>}
  n47 = OneMinus(None=ChannelMaskParameter(None=@n30) {group=NewFakeTransOverlay, maskChannel=Blue, parameterName=LightMask, defaultValue=<rgba 0 0 1 0>}.R)
  NamedRerouteDeclaration(None=Multiply(A=NamedRerouteUsage(), B=Power(Base=@n47))) {name=LightMask, nodeColor=<rgba 1 0 0.5625 1>}
  ScalarParameter() {defaultValue=2, parameterName=IdelEM}
 ]
}
