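# Animation Blueprint for the Coconut Shell backpack cosmetic, playing its idle sequence and counter-rotating the pack against the wearer's spine movement.
# Each update reads the owner mesh's spine_05 delta from reference pose relative to root and stores negated yaw as GimbalRotation roll.
# The AnimGraph feeds a one-state idle machine into additive bone-space modifiers: gimbal rotation on pack_root, then a fixed offset on pack.
# An unconnected random sequence player remains in the state.
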AnimBlueprint /BRCosmetics/Accessories/FORT_Backpacks/Backpack_M_MED_CoconutShell/Meshes/M_MED_CoconutShell_Pack_AnimBP
target_skeleton=&/M_MED_CoconutShell_Pack_Skeleton
graphs [
 AnimGraph {nodes [
   AnimGraphNode_StateMachine_0 = Animation Testing() {is=title AnimGraphNode_StateMachine}
   n1 = "Transform (Modify) Bone - Bone: pack_root"(
    ComponentPose=Local To Component(LocalPose=@AnimGraphNode_StateMachine_0)
    Alpha=1
    Translation="0,0,0"
    Rotation=Get GimbalRotation() -> (GimbalRotation:Rotator)
    Scale="1,1,1"
   ) {rotationMode=BMM_Additive, rotationSpace=BCS_BoneSpace, alphaBoolBlend {blendOption=Linear}, boneToModify {boneName=pack_root}}
   n2 = "Transform (Modify) Bone - Bone: pack"(ComponentPose=@n1, Translation="-3,0,0", Rotation="0,0,0") {
    rotationMode=BMM_Additive
    rotationSpace=BCS_BoneSpace
    scaleMode=BMM_Replace
    translationMode=BMM_Additive
    translationSpace=BCS_BoneSpace
    alphaBoolBlend {blendOption=Linear}
    alphaScaleBiasClamp {bInterpResult=true, interpSpeedDecreasing=5}
    boneToModify {boneName=pack}
    translation {x=-3}
   }
   Output Pose(Result=Component To Local(ComponentPose=@n2)) {layerGroup=None}
  ]}
 Animation Testing {
  entry=State
  path=AnimGraph.AnimGraphNode_StateMachine_0.Animation Testing
  ids=id
  nodes [AnimStateNode_0 = State() {is=title AnimStateNode}]
 }
 State {
  path=AnimGraph.AnimGraphNode_StateMachine_0.Animation Testing.AnimStateNode_0.State
  nodes [
   Random Sequence Player!() -> (Pose:Pose Link Structure) {
    is=title AnimGraphNode_RandomPlayer
    entries [
     {
      chanceToPlay=0.5
      maxLoopCount=0
      maxPlayRate=1
      minLoopCount=0
      minPlayRate=1
      sequence=/BRCosmetics/Accessories/FORT_Backpacks/Backpack_M_MED_CoconutShell/Anim/CoconutShell_Idle_pack
      blendIn {blendOption=Linear, blendTime=0.2, customCurve=null}
     }
     {
      chanceToPlay=0.5
      maxLoopCount=0
      maxPlayRate=1
      minLoopCount=0
      minPlayRate=1
      sequence=null
      blendIn {blendOption=Linear, blendTime=0.2, customCurve=null}
     }
    ]
   }
   Output Animation Pose(Result=Sequence Player 'CoconutShell_Idle_pack'() {sequence=&CoconutShell_Idle_pack})
  ]
 }
 EventGraph {nodes [
   Event BlueprintUpdateAnimation()
   n1 = GetOwnerSkeletalMeshComponent()
   n2 = BreakRotator(InRot=BreakTransform(InTransform=GetDeltaTransformFromRefPose(self=@n1, BoneName=spine_05, BaseName=root)).Rotation)
   > Set GimbalRotation(GimbalRotation:Rotator=MakeRotator(Roll=float * float(A=@n2.Yaw, B=-1) {is=title K2Node_PromotableOperator})) -> (Output_Get:Rotator)
   TryGetPawnOwner()
  ]}
]
variables [GimbalRotation {type=Rotator}]
